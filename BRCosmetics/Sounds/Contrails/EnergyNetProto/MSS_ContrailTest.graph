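# Prototype stereo MetaSound for contrail audio, switching between freefall and glide music-timed loops.
# A four-beat clock at 170.92 BPM feeds self-closing trigger gates that OnFreefall or OnGlide open, so each switch lands on the next tick.
# An ADSR sweeps a biquad filter's cutoff down on the request and back up after the switch.
# OnLand is declared but unconnected.
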MetaSoundSource /BRCosmetics/Sounds/Contrails/EnergyNetProto/MSS_ContrailTest
inputs [
 OnFreefall {type=Trigger}
 OnGlide {type=Trigger}
 OnLand {type=Trigger}
 UE.Source.OnPlay {type=Trigger}
]
interfaces [UE.OutputFormat.Stereo:1.0, UE.Source:1.1]
outputs [
 "UE.OutputFormat.Stereo.Audio:0" {type=Audio}
 "UE.OutputFormat.Stereo.Audio:1" {type=Audio}
]
pages [Default {
  input_defaults {OnFreefall=null, OnGlide=null, OnLand=null, UE.Source.OnPlay=null}
  nodes [
   n1 = UE.Multiply:Time by Float(PrimaryOperand=UE.BPMToSeconds() {also {defaults {BPM=170.92}}}.Seconds) {also {defaults {AdditionalOperands=4}}}
   n2 = VariableAccessor.Trigger(Variable=VariableMutator.Trigger(Variable=InitVariable.Trigger().Variable, Value=UE.TriggerRepeat(Start=^UE.Source.OnPlay, Period=@n1.Out).RepeatOut).Variable)
   n3 = VariableDeferredAccessor.Trigger(Variable=InitVariable.Trigger().Variable)
   TriggerControl_1 = UE.Trigger Control(Open=^OnGlide, Trigger In=@n2.Value, Close=@n3.Value)
   n4 = VariableAccessor.Trigger(Variable=VariableMutator.Trigger(Value=@TriggerControl_1.Trigger Out, Variable=@n3.Variable).Variable)
   n5 = VariableAccessor.Trigger(Variable=@n4.Variable)
   n6 = VariableDeferredAccessor.Trigger(Variable=InitVariable.Trigger().Variable)
   TriggerControl_0 = UE.Trigger Control(Open=^OnFreefall, Trigger In=VariableAccessor.Trigger(Variable=@n2.Variable).Value, Close=@n6.Value)
   n7 = VariableAccessor.Trigger(Variable=VariableMutator.Trigger(Value=@TriggerControl_0.Trigger Out, Variable=@n6.Variable).Variable)
   n8 = VariableAccessor.Trigger(Variable=@n7.Variable)
   WavePlayer_0 = UE.Wave Player:Stereo(Stop=@n5.Value, Play=@n8.Value) {also {defaults {Loop="<MetaSoundBuilderResult.SUCCEEDED: 0>", Wave Asset=&/Contrail_Freefall_Loop}}}
   WavePlayer_1 = UE.Wave Player:Stereo(Stop=@n7.Value, Play=@n4.Value) {also {defaults {Loop="<MetaSoundBuilderResult.SUCCEEDED: 0>", Wave Asset=&/Contrail_Glide_Loop}}}
   AudioMixerAudioMixerStereo2_0 = "AudioMixer.Audio Mixer (Stereo, 2)"(In 1 L=@WavePlayer_0.Out Left, In 1 R=@WavePlayer_0.Out Right, In 0 L=@WavePlayer_1.Out Left, In 0 R=@WavePlayer_1.Out Right)
   n9 = UE.BPMToSeconds() {also {defaults {BPM=170.92}}}
   TriggerAnyTriggerAny2_0 = "TriggerAny.Trigger Any (2)"(In 1=^OnFreefall, In 0=^OnGlide)
   TriggerAnyTriggerAny2_1 = "TriggerAny.Trigger Any (2)"(In 0=VariableAccessor.Trigger(Variable=@n5.Variable).Value, In 1=VariableAccessor.Trigger(Variable=@n8.Variable).Value)
   ADSREnvelopeADSREnvelope_0 = ADSR Envelope.ADSR Envelope:Float(
    Attack Time=UE.Multiply:Time by Float(PrimaryOperand=@n9.Seconds) {also {defaults {AdditionalOperands=2}}}.Out
    Release Time=UE.Multiply:Time by Float(PrimaryOperand=@n9.Seconds) {also {defaults {AdditionalOperands=4}}}.Out
    Trigger Attack=@TriggerAnyTriggerAny2_0.Out
    Trigger Release=@TriggerAnyTriggerAny2_1.Out
   ) {also {defaults {Decay Time=0, Sustain Level=1}}}
   StereoBiquadFilter_0 = &StereoBiquadFilter(
    In L=@AudioMixerAudioMixerStereo2_0.Out L
    In R=@AudioMixerAudioMixerStereo2_0.Out R
    Cutoff Frequency=MapRange.MapRange:Float(In=@ADSREnvelopeADSREnvelope_0.Out Envelope) {also {defaults {Out Range A=20000, Out Range B=100}}}.Out Value
   )
  ]
  links [
   {dst="@UE.OutputFormat.Stereo.Audio:0", src=StereoBiquadFilter_0.Out L}
   {dst="@UE.OutputFormat.Stereo.Audio:1", src=StereoBiquadFilter_0.Out R}
  ]
 }]
properties {
 assetImportData=&self:AssetImportData
 bLooping=true
 duration=10000
 numChannels=2
 outputFormat=Stereo
 sampleRate=48000
 virtualizationMode=PlayWhenSilent
 volume=2
 assetUserData [&self:AudioAssetUserData_1]
 soundSubmixSends [{}]
}
